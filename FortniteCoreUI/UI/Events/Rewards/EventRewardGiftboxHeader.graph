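# Header widget for a tournament reward giftbox, showing the event's showdown pin beside a congratulatory title and description.
# On Refresh Data from its FortShowdownDetailView parent, it formats localized text from the tournament title lines, with generic fallbacks when the first is empty.
# The description adds the round name when visual rounds exist.
# It force-refreshes the pin with series and event IDs, collapsing it when the event ID is empty.
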WidgetBlueprint /FortniteCoreUI/UI/Events/Rewards/EventRewardGiftboxHeader
graphs [EventGraph {nodes [
   Event Refresh Data() {is=title K2Node_Event}
   n1 = Break Fort Tournament Display Info!(FortTournamentDisplayInfo:"Fort Tournament Display Info Structure (by ref)"=Get TournamentDisplayInfo() -> (TournamentDisplayInfo:Fort Tournament Display Info Structure)) -> (
    TitleLine1:Text
    TitleLine2:Text
    ScheduleInfo:Text
    FlavorDescription:Text
    DetailsDescription:Text
    ShortFormatTitle:Text
    LongFormatTitle:Text
    BackgroundTitle:Text
    PinScoreRequirement:int
    PinEarnedText:Text
    BaseColor:Linear Color Structure
    PrimaryColor:Linear Color Structure
    SecondaryColor:Linear Color Structure
    HighlightColor:Linear Color Structure
    TitleColor:Linear Color Structure
    ShadowColor:Linear Color Structure
    BackgroundLeftColor:Linear Color Structure
    BackgroundRightColor:Linear Color Structure
    BackgroundTextColor:Linear Color Structure
    PosterFadeColor:Linear Color Structure
    AlertText:Text
    SeriesPointLeaderboardName:Text
    AlertType:EFortTournamentAlertType Enum
    RoundDisplayNames:Array of Texts
   ) {is=title K2Node_BreakStruct}
   > "SetText (Text)"(
    self=Get Text_Header() -> (Text_Header:Common Text Object Reference)
    InText=Select(
     Option 0:Text=Format Text!(Format="LOCTEXT(\"Reward from {Event Name Line 1} {EventName Line 2}!\")", Event Name Line 1:Text=@n1, EventName Line 2:Text=@n1.TitleLine2) -> (Result)
     Option 1:Text="LOCTEXT(\"Reward from Tournament Event!\")"
     Index=TextIsEmpty(InText=@n1)
    ) -> (ReturnValue:Text)
   )
   n2 = Break Fort Tournament Display Info(FortTournamentDisplayInfo=Get TournamentDisplayInfo())
   n3 = "integer > integer"(A=GetNumEventVisualRounds!(self:Fort Showdown Detail View Object Reference) -> (ReturnValue:int=0), B=0) {is=title K2Node_CallFunction}
   n4 = Format Text!(
    Format="LOCTEXT(\"You nailed it in {RoundDisplayName} of {TournamentNameLine1} {TournamentNameLine2}!\")"
    RoundDisplayName:Text=GetWindowRoundDisplayString!(self:Fort Showdown Detail View Object Reference) -> (ReturnValue:Text)
    TournamentNameLine1:Text=@n2
    TournamentNameLine2:Text=@n2.TitleLine2
   ) -> (Result)
   n5 = Format Text!(
    Format="LOCTEXT(\"You nailed it in {TournamentNameLine1} {TournamentNameLine2}!\")"
    TournamentNameLine1=@n2
    TournamentNameLine2=@n2.TitleLine2
   ) -> (Result)
   > "SetText (Text)"(
    self=Get Text_Description() -> (Text_Description:Common Text Object Reference)
    InText=Select(Option 0=Select(Option 0=@n5, Option 1=@n4, Index=@n3), Option 1="LOCTEXT(\"You nailed it!\")", Index=TextIsEmpty(InText=@n2))
   )
   n6 = GetEventId!(self:Fort Showdown Detail View Object Reference) -> (ReturnValue:String)
   > SetTournament!(
    execute
    self:Fort Showdown Detail View Object Reference=Get ShowdownPin() -> (ShowdownPin:Showdown Pin Object Reference)
    TournamentId:String=GetTournamentSeriesId!(self:Fort Showdown Detail View Object Reference) -> (ReturnValue:String)
    EventId:String=@n6
    bForceRefreshUI:bool=true
   ) -> (then)
   n7 = "To Text (String)"(InString=@n6)
   > SetVisibility(
    self=Get ShowdownPin()
    InVisibility=Select(Option 0:ESlateVisibility Enum=SelfHitTestInvisible, Option 1:ESlateVisibility Enum=Collapsed, Index=TextIsEmpty(InText=@n7)) -> (ReturnValue:ESlateVisibility Enum)
   )
  ]}]
variables []
widgets {
 info {inheritedWidgetCount=6, parentClass=/Script/FortniteUI.FortShowdownDetailView, rootWidgetClass=&VerticalBox, widgetCount=15}
 widgets [
  VerticalBox_3 {widgetClassPath=&VerticalBox, widget=VerticalBox_3}
  HorizontalBox_1 {
   widgetClassPath=&HorizontalBox
   parent=VerticalBox_3
   widget=HorizontalBox_1
   slot {class=VerticalBoxSlot, properties {horizontalAlignment=HAlign_Center}}
  }
  ShowdownPin {
   widgetClassPath=/FortniteCoreUI/Frontend/Showdown/ShowdownPin
   bIsVariable=true
   parent=HorizontalBox_1
   widget=ShowdownPin
   properties {dEVShowPin=true}
   slot {class=HorizontalBoxSlot}
  }
  VerticalBox_1 {
   widgetClassPath=&VerticalBox
   parent=HorizontalBox_1
   widget=VerticalBox_1
   slot {class=HorizontalBoxSlot, properties {padding=<ltrb 25 0 0 0>}}
  }
  SizeBox_0 {
   widgetClassPath=&SizeBox
   parent=VerticalBox_1
   widget=SizeBox_0
   properties {bOverride_MaxDesiredWidth=true, maxDesiredWidth=800}
   slot {class=VerticalBoxSlot}
  }
  VerticalBox_0 {widgetClassPath=&VerticalBox, parent=SizeBox_0, widget=VerticalBox_0, slot {class=SizeBoxSlot}}
  BorderItemNAme {
   widgetClassPath=&Border
   parent=VerticalBox_0
   widget=BorderItemNAme
   properties {
    horizontalAlignment=HAlign_Center
    verticalAlignment=VAlign_Center
    background {
     bIsDynamicallyLoaded=false
     drawAs=NoDrawType
     imageType=NoImage
     mirroring=NoMirror
     resourceName=None
     resourceObject=null
     tiling=NoTile
     imageSize=<32 32>
     margin=<ltrb 0 0 0 0>
     outlineSettings {
      bUseBrushTransparency=false
      roundingType=HalfHeightRadius
      width=0
      color {colorUseRule=UseColor_Specified, specifiedColor=<rgba 0 0 0 0>}
      cornerRadii=<0 0 0 0>
     }
     tintColor {colorUseRule=UseColor_Specified, specifiedColor=<rgba 1 1 1 1>}
     uVRegion {bIsValid=false, max=<0 0>, min=<0 0>}
    }
    padding=<ltrb 0 4 0 0>
   }
   slot {class=VerticalBoxSlot}
  }
  Text_Header {
   widgetClassPath=&CommonTextBlock
   bIsVariable=true
   parent=BorderItemNAme
   widget=Text_Header
   properties {
    justification=Center
    style=/FortUILibrary/Styles/Typography/Foundation/TextStyle-Header-L.TextStyle-Header-L_C
    text="[ph] Reward from {Event Name}!"
    wrapTextAt=800
    font {
     bForceMonospaced=false
     bMaterialIsStencil=false
     fontMaterial=null
     fontObject=&Burbank
     letterSpacing=0
     monospacedWidth=1
     size=48
     skewAmount=0
     typefaceFontName=Burbank-Black
     outlineSettings {
      bApplyOutlineToDropShadows=false
      bMiteredCorners=false
      bSeparateFillAlpha=false
      outlineMaterial=null
      outlineSize=0
      outlineColor=<rgba 0 0 0 1>
     }
    }
    shadowOffset=<0 0>
   }
   slot {class=BorderSlot, properties {horizontalAlignment=HAlign_Center, verticalAlignment=VAlign_Center, padding=<ltrb 0 4 0 0>}}
  }
  Text_Description {
   widgetClassPath=&CommonTextBlock
   bIsVariable=true
   parent=VerticalBox_0
   widget=Text_Description
   properties {
    justification=Center
    text="[ph] Lorem ipsum dolor sit amet, consectetur adipiscing elit. Nam dictum nisl dui, a consequat mi mattis convallis. Aenean eleifend porttitor risus, imperdiet dictum risus porta at. Maecenas "
    wrapTextAt=800
    wrappingPolicy=AllowPerCharacterWrapping
    font {
     bForceMonospaced=false
     bMaterialIsStencil=false
     fontMaterial=null
     fontObject=&BurbankSmall
     letterSpacing=0
     monospacedWidth=1
     size=24
     skewAmount=0
     typefaceFontName=NotoSans-Regular
     outlineSettings {
      bApplyOutlineToDropShadows=false
      bMiteredCorners=false
      bSeparateFillAlpha=false
      outlineMaterial=null
      outlineSize=0
      outlineColor=<rgba 0 0 0 1>
     }
    }
   }
   slot {class=VerticalBoxSlot, properties {verticalAlignment=VAlign_Bottom, padding=<ltrb 0 20 0 0>}}
  }
  Text_TimeRemaining {widgetClassPath=&CommonTextBlock, bInherited=true, widget=null}
  Text_DebugId {widgetClassPath=&TextBlock, bInherited=true, widget=null, properties=~}
  Image_PosterFront {widgetClassPath=/Script/EpicCMSUIFramework.EpicCMSImage, bInherited=true, widget=null}
  Image_PosterBack {widgetClassPath=&EpicCMSImage, bInherited=true, widget=null}
  Image_LoadingScreen {widgetClassPath=&EpicCMSImage, bInherited=true, widget=null}
  Image_PlaylistTile {widgetClassPath=&EpicCMSImage, bInherited=true, widget=null}
 ]
}
