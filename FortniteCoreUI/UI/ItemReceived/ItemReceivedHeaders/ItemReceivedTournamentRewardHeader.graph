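# Gift box header sub-widget for tournament rewards in the item-received UI, wrapping an EventRewardGiftboxHeader inside a vertical box.
# On initialization it calls SetTournament with empty IDs to force a refresh.
# InitFromGiftBoxItem_BP calls the parent, then reads the gift box item's string params and forces another refresh.
# The eventId param feeds TournamentId and eventWindowId feeds EventId, so the names are offset.
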WidgetBlueprint /FortniteCoreUI/UI/ItemReceived/ItemReceivedHeaders/ItemReceivedTournamentRewardHeader
graphs [EventGraph {nodes [
   Event OnInitialized()
   > SetTournament!(
    execute
    self:Fort Showdown Detail View Object Reference=Get EventRewardGiftboxHeader() -> (EventRewardGiftboxHeader:Event Reward Giftbox Header Object Reference)
    TournamentId:String
    EventId:String
    bForceRefreshUI:bool=true
   ) -> (then)
   Event InitFromGiftBoxItem_BP() {is=title K2Node_Event}
   > Parent: InitFromGiftBoxItem_BP() {is=title K2Node_CallParentFunction}
   n1 = Get GiftBoxItem_BP() -> (GiftBoxItem_BP:Fort Gift Box Item Object Reference)
   > SetTournament(
    self=Get EventRewardGiftboxHeader()
    TournamentId=TryGetParamString!(self:Fort Gift Box Item Object Reference=@n1, ParamName:String=eventId) -> (OutValue:String, ReturnValue:bool)
    EventId=TryGetParamString(self=@n1, ParamName=eventWindowId)
    bForceRefreshUI=true
   )
  ]}]
variables []
widgets {
 info {
  inheritedWidgetCount=0
  parentClass=/Game/UI/ItemReceived/ItemReceivedHeaders/ItemReceivedHeaderSubWidgetBase.ItemReceivedHeaderSubWidgetBase_C
  rootWidgetClass=&VerticalBox
  widgetCount=2
 }
 widgets [
  VerticalBox_3 {widgetClassPath=&VerticalBox, widget=VerticalBox_3}
  EventRewardGiftboxHeader {
   widgetClassPath=/FortniteCoreUI/UI/Events/Rewards/EventRewardGiftboxHeader
   bIsVariable=true
   parent=VerticalBox_3
   widget=EventRewardGiftboxHeader
   slot {class=VerticalBoxSlot}
  }
 ]
}
