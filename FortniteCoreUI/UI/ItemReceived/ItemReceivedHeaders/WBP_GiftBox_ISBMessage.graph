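# Info-box widget for the item-received gift box header, showing by default a notice that PCB-granted items are temporary and usable only at partnered PCBs.
# A grid layers wrapping text over a dark UIKit backplate, with a status indicator pinned to the top-left corner through a visual attachment.
# PreConstruct passes the widget's variables to helper functions that set the indicator's configuration, type, size and icon, then the message text.
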WidgetBlueprint /FortniteCoreUI/UI/ItemReceived/ItemReceivedHeaders/WBP_GiftBox_ISBMessage
graphs [
 EventGraph {nodes [
   Event PreConstruct()
   > SetStatusIndicatorProperties!(
    execute
    self
    Configuration:E_UI_StatusIndicator_Configuration Enum=Get Configuration() -> (Configuration:E_UI_StatusIndicator_Configuration Enum=NewEnumerator0)
    Type:E_UI_StatusIndicator_Type Enum=Get Type() -> (Type:E_UI_StatusIndicator_Type Enum=NewEnumerator0)
    IconSize:E_UI_StatusIndicator_Size Enum=Get Icon Size() -> (Icon Size:E_UI_StatusIndicator_Size Enum=NewEnumerator2)
    IconAsset:Object Reference=Get Icon Asset() -> (Icon Asset:Object Reference)
   ) -> (then)
   > SetInfoBoxText!(execute, self, InText:Text=Get ISBMessageText() -> (ISBMessageText:Text)) -> (then)
  ]}
 SetInfoBoxText {nodes [
   n1 = SetInfoBoxText!() -> (then, InText:Text) {is=title K2Node_FunctionEntry}
   > "SetText (Text)"(self=Get Txt_Info() -> (Txt_Info:Common Text Object Reference), InText=@n1)
  ]}
 SetStatusIndicatorProperties {nodes [
   SetStatusIndicatorProperties!() -> (
    then
    Configuration:E_UI_StatusIndicator_Configuration Enum=NewEnumerator0
    Type:E_UI_StatusIndicator_Type Enum=NewEnumerator0
    IconSize:E_UI_StatusIndicator_Size Enum=NewEnumerator2
    IconAsset:Object Reference
   ) {is=title K2Node_FunctionEntry}
   n1 = Get WBP_UIKit_StatusIndicator() -> (WBP_UIKit_StatusIndicator:WBP UIKit Status Indicator Object Reference)
   > Set Configuration!(
    execute
    Configuration:E_UI_StatusIndicator_Configuration Enum=Get Configuration() -> (Configuration=NewEnumerator0)
    self:WBP UIKit Status Indicator Base Object Reference=@n1
   ) -> (then, Output_Get:E_UI_StatusIndicator_Configuration Enum=NewEnumerator0)
   > Set Type!(execute, Type:E_UI_StatusIndicator_Type Enum=Get Type() -> (Type=NewEnumerator0), self:WBP UIKit Status Indicator Base Object Reference=@n1) -> (then, Output_Get:E_UI_StatusIndicator_Type Enum=NewEnumerator0)
   > Set IconSize!(
    execute
    IconSize:E_UI_StatusIndicator_Size Enum=Get IconSize() -> (IconSize:E_UI_StatusIndicator_Size Enum=NewEnumerator2)
    self:WBP UIKit Status Indicator Base Object Reference=@n1
   ) -> (then, Output_Get:E_UI_StatusIndicator_Size Enum=NewEnumerator2)
   > SetIcon!(
    execute
    self:WBP UIKit Status Indicator Base Object Reference=@n1
    IconAsset:Object Reference=Get IconAsset() -> (IconAsset:Object Reference)
   ) -> (then)
  ]}
]
variables [
 Configuration {type=E_UI_StatusIndicator_Configuration, default=Icon}
 Icon Asset {type=Object, default=/FortUILibrary/Textures/MaterialInstances/MI_UIKit_Icon_UpdatePending_Small}
 Icon Size {type=E_UI_StatusIndicator_Size, default=Default}
 ISBMessageText {type=string, default=Items granted via PCB benefits are temporary only and accessible only at partnered PCBs.}
 Type {type=E_UI_StatusIndicator_Type, default=Utility}
]
widgets {
 info {inheritedWidgetCount=0, parentClass=&UserWidget, rootWidgetClass=&GridPanel, widgetCount=5}
 widgets [
  GP_Content {widgetClassPath=&GridPanel, widget=GP_Content, properties {columnFill [1], rowFill [1]}}
  WBP_UIKit_Backplate {
   widgetClassPath=&WBP_UIKit_Backplate
   parent=GP_Content
   widget=WBP_UIKit_Backplate
   properties {brightness=Dark}
   slot {class=GridSlot}
  }
  Txt_Info {
   widgetClassPath=&CommonTextBlock
   bIsVariable=true
   parent=GP_Content
   widget=Txt_Info
   properties {
    autoWrapText=true
    text=Items granted via PCB benefits are temporary only and accessible only at partnered PCBs.
    colorAndOpacity {colorUseRule=UseColor_Specified, specifiedColor=<rgba 0.5776 0.5776 0.5776 1>}
    font {
     bForceMonospaced=false
     bMaterialIsStencil=false
     fontMaterial=null
     fontObject=&NotoSans
     letterSpacing=0
     monospacedWidth=1
     size=24
     skewAmount=0
     typefaceFontName=NotoSans-Regular
     outlineSettings {
      bApplyOutlineToDropShadows=false
      bMiteredCorners=false
      bSeparateFillAlpha=false
      outlineMaterial=null
      outlineSize=0
      outlineColor=<rgba 0 0 0 1>
     }
    }
    shadowOffset=<0 0>
   }
   slot {class=GridSlot, properties {layer=10, verticalAlignment=VAlign_Top, padding=<ltrb 12 8 12 8>}}
  }
  VA_StatusIndicator {
   widgetClassPath=&FortVisualAttachment
   parent=GP_Content
   widget=VA_StatusIndicator
   properties {contentAnchor=<0.5 0.5>}
   slot {class=GridSlot, properties {horizontalAlignment=HAlign_Left, layer=20, verticalAlignment=VAlign_Top}}
  }
  WBP_UIKit_StatusIndicator {
   widgetClassPath=/FortUILibrary/UIKit/StatusIndicator/WBP_UIKit_StatusIndicator
   bIsVariable=true
   parent=VA_StatusIndicator
   widget=WBP_UIKit_StatusIndicator
   properties {
    type=Utility
    bangType=Custom
    icon=/FortUILibrary/Textures/MaterialInstances/MI_UIKit_Icon_UpdatePending
    overrideStatusIndicatorState=true
    overrideStatusIndicatorVisibility=true
   }
   slot {class=SizeBoxSlot, properties {horizontalAlignment=HAlign_Center, verticalAlignment=VAlign_Center}}
  }
 ]
}
